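# Niagara burst effect for the mouse despawn in the BurgleBoss cosmetic companion emote, a short warm gold puff built from three sprite emitters.
# Glow is one flash whose size, color, corona and airglow parameters are animated by curves over its lifetime.
# GoldDust scatters growing, fading dust slowed by drag, while ThickStars throws small hue-shifted stars with waveform-driven size and curl noise.
# Front-end modifiers and the bIsFrontEnd/bIsFrontEndPreview flags adjust brightness between gameplay and menus.
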NiagaraSystem /CosmeticCompanions/FX/Builtins/BurgleBoss/Emote/NS_Emote_BurgleBoss_MouseDespawn
dependencies {
 usedDataInterfaces [&NiagaraDataInterfaceArrayColor, &NiagaraDataInterfaceCurve, &NiagaraDataInterfaceColorCurve]
 usedDynamicInputs []
 usedModules [
  &self:SystemSpawnScript.SystemScriptSource.SystemScriptGraph.NiagaraNodeAssignment_0.SetVariables_#1
  &self:SystemSpawnScript.SystemScriptSource.SystemScriptGraph.NiagaraNodeAssignment_1.SetVariables_#2
  &SystemState
 ]
}
emitters {
 Glow {
  emitterUpdateScript [
   &EmitterState()
   &SpawnBurst_Instantaneous(Use Spawn Probability=false, UseLoopCountLimit=false)
  ]
  particleSpawnScript [
   &InitializeParticle(
    Color=<rgba 1 0.2509 0.0954 1>
    Lifetime=0.4
    Sprite Size Mode=::NewEnumerator3
    Uniform Sprite Size=&Multiply_Float(A=35, B=$System.ScaleMultiplier)
    UsePositionOffset=false
    Write Material Random=true
   )
   &CameraOffset(Camera Offset Amount=30)
   /SetVariables(
    Particles.AirglowColor:LinearColor=<rgba 1 0.7448 0.2279 1>
    Particles.AirglowSize:Vector2f=&Vector2DFromFloat(Value=130)
   ) {isSetParameters=true}
  ]
  particleUpdateScript [
   &ParticleState()
   &ScaleSpriteSize(
    Uniform Curve Scale=5
    Uniform Curve Sprite Scale=DataInterface() {propertyValues {Curve {
       defaultValue=0.25
       keys [
        {arriveTangent=4.1785, interpMode=RCIM_Cubic, leaveTangent=4.1785, tangentMode=RCTM_User, time=0, value=0}
        {arriveTangent=1.1179, interpMode=RCIM_Cubic, leaveTangent=1.1179, tangentMode=RCTM_User, time=0.2042, value=0.6228}
        {arriveTangent=0.0224, interpMode=RCIM_Cubic, leaveTangent=0.0224, tangentMode=RCTM_User, time=1, value=1}
       ]
      }}}
   )
   &ScaleColor(
    Linear Color Curve=DataInterface() {propertyValues {
      AlphaCurve {keys [<at 0.1074 0.5>,<at 0.9199 0>]}
      BlueCurve {keys [<at 0 0>,<at 0.3416 1>,<at 0.6368 1>,<at 1 0>]}
      GreenCurve {keys [<at 0 5.2558>,<at 0.3416 1>,<at 0.6368 1>,<at 1 0.3697>]}
      RedCurve {keys [<at 0 10>,<at 0.3416 1>,<at 0.6368 1>,<at 1 1>]}
     }}
    Scale Mode=2
   )
   /SetVariables(
    Particles.CoronaColor:LinearColor=&MultiplyLinearColor(
     Linear Color=$Particles.Initial.CoronaColor
     Scale Factor=&MakeVector4FromVectorAndFloat(
      "Vector (RGB)"=&MakeVectorFromLinearColorRGB(Color=&ColorFromCurve(ColorCurve=DataInterface() {propertyValues {
          AlphaCurve {keys []}
          BlueCurve {keys [<at 0 0.02 RCIM_Cubic>,<at 0.0076 3>,<at 0.0485 2>,<at 0.2581 1>,<at 0.5666 0.0012>,<at 1 0.067 RCIM_Cubic>]}
          GreenCurve {keys [<at 0 0.579 RCIM_Cubic>,<at 0.0076 3>,<at 0.0485 2>,<at 0.2581 1>,<at 0.5666 0.6362>,<at 1 0.1761 RCIM_Cubic>]}
          RedCurve {keys [<at 0 1 RCIM_Cubic>,<at 0.0076 3>,<at 0.0485 2>,<at 0.2581 1>,<at 0.5666 1>,<at 1 1 RCIM_Cubic>]}
         }}))
      "Float (Alpha)"=&FloatFromCurve(
       FloatCurve=DataInterface() {propertyValues {Curve {
          defaultValue=0.25
          keys [
           {arriveTangent=-3.5471, interpMode=RCIM_Cubic, leaveTangent=-3.5471, tangentMode=RCTM_User, time=0.0252, value=1}
           {arriveTangent=-1.0258, interpMode=RCIM_Cubic, leaveTangent=-1.0258, time=0.2244, value=0.385}
           {interpMode=RCIM_Cubic, tangentMode=RCTM_User, time=1, value=0}
          ]
         }}}
       Scale Curve=&FN_FrontEndFloatModifier(Gameplay=1, Front End=0.35)
      )
     )
    )
    Particles.CoronaSize:Vector2f=&Multiply_Vector2D(
     A=$Particles.Initial.CoronaSize
     B=&Vector2DFromFloat(Value=&FloatFromCurve(FloatCurve=DataInterface() {propertyValues {Curve {
          defaultValue=0
          keys [
           {arriveTangent=4.1785, interpMode=RCIM_Cubic, leaveTangent=4.1785, tangentMode=RCTM_User, time=0, value=0.25}
           {arriveTangent=1.1179, interpMode=RCIM_Cubic, leaveTangent=1.1179, tangentMode=RCTM_User, time=0.2, value=0.7}
           {arriveTangent=0.0224, interpMode=RCIM_Cubic, leaveTangent=0.0224, tangentMode=RCTM_User, time=1, value=1}
          ]
         }}}))
    )
   ) {isSetParameters=true}
   /SetVariables(
    Particles.AirglowColor:LinearColor=&MultiplyLinearColor(
     Linear Color=$Particles.Initial.AirglowColor
     Scale Factor=&Vector4FromCurve(Vector4Curve=DataInterface() {propertyValues {
        AlphaCurve {keys [<at 0.1144 0.3>,<at 0.7072 0>]}
        BlueCurve {keys [<at 0.0475 5>,<at 0.2931 1>,<at 0.6545 0.2546>]}
        GreenCurve {keys [<at 0.0475 5>,<at 0.2931 1>,<at 0.6545 0.3794>]}
        RedCurve {keys [<at 0.0475 5>,<at 0.2931 1>,<at 0.6545 1>]}
       }})
    )
    Particles.AirglowSize:Vector2f=&Multiply_Vector2D(
     A=$Particles.AirglowSize
     B=&Vector2DFromFloat(Value=&FloatFromCurve(
       FloatCurve=DataInterface() {propertyValues {Curve {
          defaultValue=0
          keys [
           {interpMode=RCIM_Cubic, tangentMode=RCTM_User, time=0.1, value=1}
           {arriveTangent=-3.3254, interpMode=RCIM_Cubic, leaveTangent=-3.3254, tangentMode=RCTM_User, time=0.2494, value=0.5465}
           {interpMode=RCIM_Cubic, tangentMode=RCTM_User, time=0.6737, value=0}
          ]
         }}}
       Scale Curve=&FN_FrontEndFloatModifier(Gameplay=0.4, Front End=0.3)
      ))
    )
   ) {isSetParameters=true}
   /SetVariables(Particles.float:NiagaraFloat=&FloatFromCurve(
     FloatCurve=DataInterface() {propertyValues {Curve {
        defaultValue=0.25
        keys [
         {arriveTangent=-3.5471, interpMode=RCIM_Cubic, leaveTangent=-3.5471, tangentMode=RCTM_User, time=0.0252, value=1}
         {arriveTangent=-1.0258, interpMode=RCIM_Cubic, leaveTangent=-1.0258, time=0.2244, value=0.385}
         {interpMode=RCIM_Cubic, tangentMode=RCTM_User, time=1, value=0}
        ]
       }}}
     Scale Curve=1.5
    )) {isSetParameters=true}
  ]
  rendererClasses [&NiagaraSpriteRendererProperties]
  renderers [
   {rendererClass=&NiagaraSpriteRendererProperties}
   {rendererClass=&NiagaraSpriteRendererProperties, rendererIndex=1}
  ]
 }
 GoldDust {
  emitterUpdateScript [
   &EmitterState(Life Cycle Mode=::NewEnumerator1, Loop Behavior=::NewEnumerator1, Loop Duration=3, UseLoopDelay=false)
   &SpawnBurst_Instantaneous(Spawn Count=13, Use Spawn Probability=false, UseLoopCountLimit=false)
  ]
  particleSpawnScript [
   &InitializeParticle(
    Color=&Multiply_LinearColorByFloat(
     Linear Color=&SelectLinearColorFromArray(
      Array Sampling Mode=::NewEnumerator0
      Direct Array Index=?
      Color Selection Array=DataInterface() {propertyValues {ColorData [<rgba 8 4.0072 0.7917 0.1>, <rgba 8 5.6344 1.3957 0.1>, <rgba 8 2.2178 0.0178 0.1>, <rgba 8 3.6879 0.2005 0.1>, <rgba 8 4.0072 0.7917 0.1>]}}
      Random Seed=0
     )
     Float=&FN_FrontEndFloatModifier(Gameplay=1, Front End=0.73)
    )
    Lifetime=&UniformAOrBFloat(
     A=&RandomRangeFloat(Minimum=0.5, Maximum=0.8)
     B=&RandomRangeFloat(Minimum=1, Maximum=1.2)
     Distribution Weight=0.3
     Evaluation Type=::NewEnumerator0
     Recalculate Random Each Loop=true
     Override Seed=false
    )
    Sprite Rotation Mode=::NewEnumerator2
    Sprite Size Mode=::NewEnumerator3
    Uniform Sprite Size=&Multiply_Float(
     A=&RandomRangeFloat(Minimum=18, Maximum=30)
     B=&RandomRangeFloat(Minimum=1, Maximum=2.5)
    )
    UsePositionOffset=false
    Write Material Random=true
   )
   &CameraOffset(Camera Offset Amount=&RandomRangeFloat(Minimum=37, Maximum=50))
   &ShapeLocation(MaskSpawnGroup=false, Non Uniform Scale=<1 1 2.5>, Override Seed=false, Sphere Radius=1, UseOwnerScale=true)
   &AddVelocity(Module.Use Falloff Distance=false, Override Seed=false, Velocity Mode=::NewEnumerator1)
   &ApplyInitialForces()
  ]
  particleUpdateScript [
   &ParticleState()
   &ScaleSpriteSize(Uniform Curve Sprite Scale=DataInterface() {propertyValues {Curve {
       defaultValue=0
       keys [
        {arriveTangent=1.1179, interpMode=RCIM_Cubic, leaveTangent=1.1179, tangentMode=RCTM_User, time=0, value=0.35}
        {arriveTangent=0.0224, interpMode=RCIM_Cubic, leaveTangent=0.0224, tangentMode=RCTM_User, time=0.7686, value=1.0094}
        {arriveTangent=1.402, interpMode=RCIM_Cubic, leaveTangent=1.402, tangentMode=RCTM_User, time=0.9978, value=1.46}
       ]
      }}})
   &ScaleColor(
    Scale Alpha=&FloatFromCurve(
     FloatCurve=DataInterface() {propertyValues {Curve {
        defaultValue=0
        keys [
         {arriveTangent=-3.0035, interpMode=RCIM_Cubic, leaveTangent=-3.0035, tangentMode=RCTM_User, time=0.0118, value=1.0048}
         {arriveTangent=-1.2935, interpMode=RCIM_Cubic, leaveTangent=-1.2935, tangentMode=RCTM_User, time=0.3271, value=0.3741}
         {arriveTangent=-0.0404, interpMode=RCIM_Cubic, leaveTangent=-0.0404, tangentMode=RCTM_User, time=1, value=0}
        ]
       }}}
     Scale Curve=&FN_FrontEndFloatModifier(
      Gameplay=&RandomRangeFloat(Minimum=2.5, Maximum=6)
      Front End=&RandomRangeFloat(Minimum=2, Maximum=4)
     )
    )
    Scale RGB=&MakeVectorFromLinearColorRGB(Color=&MakeCustomLinearColorFromBool(
      Bool=&Boolean_OrOperation(Bool1=$User.bIsFrontEnd, Bool2=$User.bIsFrontEndPreview)
      True Color=&ColorFromCurve(ColorCurve=DataInterface() {propertyValues {
         AlphaCurve {keys []}
         BlueCurve {keys [<at 0.1264 1.5>, {arriveTangent=-2.5049, interpMode=RCIM_Cubic, leaveTangent=-2.5049, time=0.2304, value=1}, <at 0.3261 1 RCIM_Cubic>]}
         GreenCurve {keys [<at 0.1264 1.5>, {arriveTangent=-2.5049, interpMode=RCIM_Cubic, leaveTangent=-2.5049, time=0.2304, value=1}, <at 0.3261 1 RCIM_Cubic>]}
         RedCurve {keys [<at 0.1264 1.5>, {arriveTangent=-2.5049, interpMode=RCIM_Cubic, leaveTangent=-2.5049, time=0.2304, value=1}, <at 0.3261 1 RCIM_Cubic>]}
        }})
      False Color=&ColorFromCurve(ColorCurve=DataInterface() {propertyValues {
         AlphaCurve {keys []}
         BlueCurve {keys [<at 0.1264 3>, {interpMode=RCIM_Cubic, leaveTangent=-2.2895, time=0.2304, value=1}, <at 0.3261 1 RCIM_Cubic>]}
         GreenCurve {keys [<at 0.1264 3>, {interpMode=RCIM_Cubic, leaveTangent=-2.2895, time=0.2304, value=1}, <at 0.3261 1 RCIM_Cubic>]}
         RedCurve {keys [<at 0.1264 3>, {interpMode=RCIM_Cubic, leaveTangent=-2.2895, time=0.2304, value=1}, <at 0.3261 1 RCIM_Cubic>]}
        }})
     ))
    ScaleA=true
    ScaleRGB=true
   )
   &Drag(Use Linear Drag=true, Use Rotational Drag=true)
   &SolveForcesAndVelocity(Clamp Velocity=false, Limit Acceleration=false)
  ]
  rendererClasses [&NiagaraSpriteRendererProperties]
  renderers [{rendererClass=&NiagaraSpriteRendererProperties}]
 }
 ThickStars {
  emitterUpdateScript [
   &EmitterState(Life Cycle Mode=::NewEnumerator1, Loop Behavior=::NewEnumerator1, Loop Duration=1, UseLoopDelay=false)
   &SpawnBurst_Instantaneous(Spawn Count=5, Use Spawn Probability=false, UseLoopCountLimit=false)
  ]
  particleSpawnScript [
   &InitializeParticle(
    AdjustAlpha=false
    AdjustHue=true
    AdjustSaturation=false
    AdjustValue=false
    Color=<rgba 5 2.0805 0.2401 1>
    Color Mode=::NewEnumerator3
    Hue Shift Range:Vector2f=<-0.05 0.05>
    Lifetime Max=0.7
    Lifetime Min=0.3
    Lifetime Mode=::NewEnumerator1
    Sprite Size Mode=::NewEnumerator3
    Uniform Sprite Size=&Multiply_Float(
     A=&RandomRangeFloat(Minimum=4, Maximum=11)
     B=&RandomRangeFloat(Minimum=0.8, Maximum=1.6)
    )
    UsePositionOffset=false
    Write Material Random=true
   )
   &CameraOffset(Camera Offset Amount=$System.OffsetCam)
   &ShapeLocation(
    MaskSpawnGroup=false
    Override Seed=false
    Sphere Radius=&Multiply_Float(A=1, B=&RandomRangeFloat(Minimum=1, Maximum=2))
    UseOwnerScale=true
   )
   &AddVelocity(
    Module.Use Falloff Distance=false
    Override Seed=false
    Velocity Mode=::NewEnumerator1
    Velocity Speed=&UniformAOrBFloat(
     A=&RandomRangeFloat(Minimum=100, Maximum=150)
     B=250
     Distribution Weight=0.3
     Evaluation Type=::NewEnumerator0
     Recalculate Random Each Loop=true
     Override Seed=false
    )
   )
   &ApplyInitialForces()
  ]
  particleUpdateScript [
   &ParticleState()
   &ScaleSpriteSize(
    Scale Sprite Size Mode=::NewEnumerator0
    Uniform Scale Factor=&Waveform(Global Phase=$Particles.NormalizedAge, "Waveform [1]"=::NewEnumerator2, "[1] Frequency"=1, "[1] Amplitude Min/Max"=<0.1 1>)
   )
   &CurlNoiseForce(
    Bias Noise=false
    Noise Frequency=15
    Noise Strength=&RandomRangeFloat(Minimum=60, Maximum=120)
    Pan Noise=false
   )
   &Drag(Drag=&RandomRangeFloat(Minimum=7, Maximum=9), Use Linear Drag=true, Use Rotational Drag=true)
   &SolveForcesAndVelocity(Clamp Velocity=false, Limit Acceleration=false)
  ]
  rendererClasses [&NiagaraSpriteRendererProperties]
  renderers [{rendererClass=&NiagaraSpriteRendererProperties}]
 }
}
summary {
 systemName=NS_Emote_BurgleBoss_MouseDespawn
 emitters [
  {emitterName=GoldDust, rendererClasses [&NiagaraSpriteRendererProperties]}
  {emitterName=Glow, rendererClasses [&NiagaraSpriteRendererProperties]}
  {emitterName=ThickStars, rendererClasses [&NiagaraSpriteRendererProperties]}
 ]
 userVariables [
  User.bIsFrontEnd {type=NiagaraBool}
  User.bIsFrontEndPreview {type=NiagaraBool}
 ]
}
