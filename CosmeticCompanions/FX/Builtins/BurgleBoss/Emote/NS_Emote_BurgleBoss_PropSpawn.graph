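# Niagara system for the BurgleBoss emote's prop spawn: a short, warm gold-orange burst of glow, mesh flash and star sparkles.
# Glow layers core, corona and airglow sprites from one curve-animated particle; MainBurst pairs a growing mesh flash with drifting, fading sprites.
# StarBits throws three sprite groups shaped by drag, noise and gravity.
# Spawn groups with MergeHelper switches let emitters share several looks, while front-end flags adjust brightness and opacity.
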NiagaraSystem /CosmeticCompanions/FX/Builtins/BurgleBoss/Emote/NS_Emote_BurgleBoss_PropSpawn
dependencies {
 usedDataInterfaces [&NiagaraDataInterfaceCurve, &NiagaraDataInterfaceColorCurve, &NiagaraDataInterfaceArrayColor]
 usedDynamicInputs []
 usedModules [
  &self:SystemSpawnScript.SystemScriptSource.SystemScriptGraph.NiagaraNodeAssignment_1.SetVariables_#1
  &self:SystemSpawnScript.SystemScriptSource.SystemScriptGraph.NiagaraNodeAssignment_2.SetVariables_#2
  &SystemState
 ]
}
emitters {
 Glow {
  emitterUpdateScript [
   &EmitterState()
   &SpawnBurst_Instantaneous(Use Spawn Probability=false, UseLoopCountLimit=false)
  ]
  particleSpawnScript [
   &InitializeParticle(
    Color=<rgba 1 0.2509 0.0954 1>
    Lifetime=0.4
    Sprite Size Mode=::NewEnumerator3
    Uniform Sprite Size=&Multiply_Float(A=35, B=$System.ScaleMultiplier)
    UsePositionOffset=false
    Write Material Random=true
   )
   &CameraOffset(Camera Offset Amount=30)
   /SetVariables(
    Particles.CoronaColor:LinearColor=<rgba 2 0.7617 0.0695 1>
    Particles.CoronaSize:Vector2f=&Vector2DFromFloat(Value=&Multiply_Float(A=30, B=$System.ScaleMultiplier))
   ) {isSetParameters=true}
   /SetVariables(
    Particles.AirglowColor:LinearColor=<rgba 1 0.7448 0.2279 1>
    Particles.AirglowSize:Vector2f=&Vector2DFromFloat(Value=130)
   ) {isSetParameters=true}
  ]
  particleUpdateScript [
   &ParticleState()
   &ScaleSpriteSize(Uniform Curve Sprite Scale=DataInterface() {propertyValues {Curve {
       defaultValue=0.25
       keys [
        {arriveTangent=4.1785, interpMode=RCIM_Cubic, leaveTangent=4.1785, tangentMode=RCTM_User, time=0, value=0}
        {arriveTangent=1.1179, interpMode=RCIM_Cubic, leaveTangent=1.1179, tangentMode=RCTM_User, time=0.2042, value=0.6228}
        {arriveTangent=0.0224, interpMode=RCIM_Cubic, leaveTangent=0.0224, tangentMode=RCTM_User, time=1, value=1}
       ]
      }}})
   &ScaleColor(
    Color Scale:LinearColor=&MultiplyLinearColor(
     Linear Color=&ColorFromCurve(ColorCurve=DataInterface() {propertyValues {
        AlphaCurve {keys [<at 0.1074 0.75>,<at 0.9199 0>]}
        BlueCurve {keys [<at 0 0>,<at 0.3416 1>,<at 0.6368 1>,<at 1 0>]}
        GreenCurve {keys [<at 0 5.2558>,<at 0.3416 1>,<at 0.6368 1>,<at 1 0.3697>]}
        RedCurve {keys [<at 0 10>,<at 0.3416 1>,<at 0.6368 1>,<at 1 1>]}
       }})
     Scale Factor=&MakeVector4FromVectorAndFloat("Vector (RGB)"=<1 1 1>, "Float (Alpha)"=&FN_FrontEndFloatModifier(Gameplay=1, Front End=0.3))
    )
    Scale Mode=3
   )
   /SetVariables(
    Particles.CoronaColor:LinearColor=&MultiplyLinearColor(
     Linear Color=$Particles.Initial.CoronaColor
     Scale Factor=&MakeVector4FromVectorAndFloat(
      "Vector (RGB)"=&MakeVectorFromLinearColorRGB(Color=&ColorFromCurve(ColorCurve=DataInterface() {propertyValues {
          AlphaCurve {keys []}
          BlueCurve {keys [<at 0 0.02 RCIM_Cubic>,<at 0.0076 3>,<at 0.0485 2>,<at 0.2581 1>,<at 0.5666 0.0012>,<at 1 0.067 RCIM_Cubic>]}
          GreenCurve {keys [<at 0 0.579 RCIM_Cubic>,<at 0.0076 3>,<at 0.0485 2>,<at 0.2581 1>,<at 0.5666 0.6362>,<at 1 0.1761 RCIM_Cubic>]}
          RedCurve {keys [<at 0 1 RCIM_Cubic>,<at 0.0076 3>,<at 0.0485 2>,<at 0.2581 1>,<at 0.5666 1>,<at 1 1 RCIM_Cubic>]}
         }}))
      "Float (Alpha)"=&FloatFromCurve(FloatCurve=DataInterface() {propertyValues {Curve {
          defaultValue=0.25
          keys [
           {arriveTangent=-3.5471, interpMode=RCIM_Cubic, leaveTangent=-3.5471, tangentMode=RCTM_User, time=0.0252, value=1}
           {arriveTangent=-1.0258, interpMode=RCIM_Cubic, leaveTangent=-1.0258, time=0.2244, value=0.385}
           {interpMode=RCIM_Cubic, tangentMode=RCTM_User, time=1, value=0}
          ]
         }}})
     )
    )
    Particles.CoronaSize:Vector2f=&Multiply_Vector2D(
     A=$Particles.Initial.CoronaSize
     B=&Vector2DFromFloat(Value=&FloatFromCurve(FloatCurve=DataInterface() {propertyValues {Curve {
          defaultValue=0
          keys [
           {arriveTangent=4.1785, interpMode=RCIM_Cubic, leaveTangent=4.1785, tangentMode=RCTM_User, time=0, value=0.25}
           {arriveTangent=1.1179, interpMode=RCIM_Cubic, leaveTangent=1.1179, tangentMode=RCTM_User, time=0.2, value=0.7}
           {arriveTangent=0.0224, interpMode=RCIM_Cubic, leaveTangent=0.0224, tangentMode=RCTM_User, time=1, value=1}
          ]
         }}}))
    )
   ) {isSetParameters=true}
   /SetVariables(
    Particles.AirglowColor:LinearColor=&MultiplyLinearColor(
     Linear Color=&MultiplyLinearColor(
      Linear Color=$Particles.Initial.AirglowColor
      Scale Factor=&Vector4FromCurve(Vector4Curve=DataInterface() {propertyValues {
         AlphaCurve {keys [<at 0.0039 0.5>,<at 0.1144 1>,<at 0.3194 0.5>,<at 0.7072 0>]}
         BlueCurve {keys [<at 0.0476 2>,<at 0.2931 1>,<at 0.6545 0.2546>]}
         GreenCurve {keys [<at 0.0476 2>,<at 0.2931 1>,<at 0.6545 0.3794>]}
         RedCurve {keys [<at 0.0476 2>,<at 0.2931 1>,<at 0.6545 1>]}
        }})
     )
     Scale Factor=&MakeVector4FromVectorAndFloat("Vector (RGB)"=<1 1 1>, "Float (Alpha)"=&FN_FrontEndFloatModifier(Gameplay=1, Front End=0.65))
    )
    Particles.AirglowSize:Vector2f=&Multiply_Vector2D(
     A=$Particles.AirglowSize
     B=&Vector2DFromFloat(Value=&FloatFromCurve(FloatCurve=DataInterface() {propertyValues {Curve {
          defaultValue=0
          keys [
           {interpMode=RCIM_Cubic, tangentMode=RCTM_User, time=0.1, value=1}
           {arriveTangent=-1.6712, interpMode=RCIM_Cubic, leaveTangent=-1.6712, tangentMode=RCTM_User, time=0.2982, value=0.8195}
           {interpMode=RCIM_Cubic, tangentMode=RCTM_User, time=0.6737, value=0}
          ]
         }}}))
    )
   ) {isSetParameters=true}
   /SetVariables(Particles.float:NiagaraFloat=&FloatFromCurve(FloatCurve=DataInterface() {propertyValues {Curve {
        defaultValue=0.25
        keys [
         {arriveTangent=-3.5471, interpMode=RCIM_Cubic, leaveTangent=-3.5471, tangentMode=RCTM_User, time=0.0252, value=1}
         {arriveTangent=-1.0258, interpMode=RCIM_Cubic, leaveTangent=-1.0258, time=0.2244, value=0.385}
         {interpMode=RCIM_Cubic, tangentMode=RCTM_User, time=1, value=0}
        ]
       }}})) {isSetParameters=true}
   /SetVariables() {isSetParameters=true}
  ]
  rendererClasses [&NiagaraSpriteRendererProperties]
  renderers [
   {rendererClass=&NiagaraSpriteRendererProperties}
   {rendererClass=&NiagaraSpriteRendererProperties, rendererIndex=1}
   {rendererClass=&NiagaraSpriteRendererProperties, rendererIndex=2}
  ]
 }
 MainBurst {
  emitterUpdateScript [
   &EmitterState(Life Cycle Mode=::NewEnumerator1, Loop Behavior=::NewEnumerator1, Loop Duration=3, UseLoopDelay=false)
   &SpawnBurst_Instantaneous(Use Spawn Probability=false, UseLoopCountLimit=false)
   &SpawnBurst_Instantaneous(Spawn Count=25, Spawn Group=1, Use Spawn Probability=false, UseLoopCountLimit=false) {module=SpawnBurst_Instantaneous001}
  ]
  particleSpawnScript [
   /SetVariables(Particles.VisibilityTag=&SpawnGroup()) {isSetParameters=true}
   &InitializeParticle(
    Color=&MergeHelper_ColorSwitch(
     0=<rgba 1 0.6385 0.1406 0.8435>
     1=&Multiply_LinearColorByFloat(
      Linear Color=&SelectLinearColorFromArray(
       Array Sampling Mode=::NewEnumerator0
       Direct Array Index=?
       Color Selection Array=DataInterface() {propertyValues {ColorData [<rgba 8 4.0072 0.7917 0.1>, <rgba 8 5.6344 1.3957 0.1>, <rgba 8 2.2178 0.0178 0.1>, <rgba 8 3.6879 0.2005 0.1>, <rgba 8 4.0072 0.7917 0.1>]}}
       Random Seed=0
      )
      Float=&FN_FrontEndFloatModifier(Gameplay=1, Front End=0.8)
     )
     2=<rgba 0 0 0 0>
     3=<rgba 0 0 0 0>
     4=<rgba 0 0 0 0>
     Index=&SpawnGroup()
     Num=::NewEnumerator3
    )
    Lifetime=&MergeHelper_FloatSwitch(
     0=0.2
     1=&UniformAOrBFloat(
      A=&RandomRangeFloat(Minimum=0.5, Maximum=0.8)
      B=&RandomRangeFloat(Minimum=1, Maximum=1.2)
      Distribution Weight=0.3
      Evaluation Type=::NewEnumerator0
      Recalculate Random Each Loop=true
      Override Seed=false
     )
     2=0
     3=0
     4=0
     Index=&SpawnGroup()
     Num=::NewEnumerator3
    )
    Mesh Scale Mode=::NewEnumerator3
    Mesh Uniform Scale=&MergeHelper_FloatSwitch(0=1.2, 1=0, 2=0, 3=0, 4=0, Index=&SpawnGroup(), Num=::NewEnumerator3)
    Sprite Size Mode=::NewEnumerator3
    Uniform Sprite Size=&MergeHelper_FloatSwitch(
     0=50
     1=&Multiply_Float(
      A=&RandomRangeFloat(Minimum=35, Maximum=50)
      B=&RandomRangeFloat(Minimum=1, Maximum=2.5)
     )
     2=0
     3=0
     4=0
     Index=&SpawnGroup()
     Num=::NewEnumerator3
    )
    UsePositionOffset=false
    Write Material Random=true
   )
   &CameraOffset(Camera Offset Amount=&MergeHelper_FloatSwitch(
     0=5
     1=&RandomRangeFloat(Minimum=37, Maximum=50)
     2=0
     3=0
     4=0
     Index=&SpawnGroup()
     Num=::NewEnumerator3
    ))
   &ShapeLocation(
    MaskSpawnGroup=false
    Override Seed=false
    Sphere Radius=&MergeHelper_FloatSwitch(0=0, 1=40, 2=0, 3=0, 4=0, Index=&SpawnGroup(), Num=::NewEnumerator3)
    UseOwnerScale=true
   )
   &AddVelocity(
    Module.Use Falloff Distance=false
    Override Seed=false
    Velocity Mode=::NewEnumerator1
    Velocity Speed=&MergeHelper_FloatSwitch(0=0, 1=25, 2=0, 3=0, 4=0, Index=&SpawnGroup(), Num=::NewEnumerator3)
   )
  ]
  particleUpdateScript [
   &ParticleState()
   &ScaleSpriteSize(
    Scale Factor=&MergeHelper_Vector2Switch(
     0=<1 1>
     1=&Vector2DFromFloat(Value=&FloatFromCurve(FloatCurve=DataInterface() {propertyValues {Curve {
          defaultValue=0
          keys [
           {arriveTangent=1.1179, interpMode=RCIM_Cubic, leaveTangent=1.1179, tangentMode=RCTM_User, time=0, value=0.35}
           {arriveTangent=0.0224, interpMode=RCIM_Cubic, leaveTangent=0.0224, tangentMode=RCTM_User, time=0.7686, value=1.0094}
           {arriveTangent=1.402, interpMode=RCIM_Cubic, leaveTangent=1.402, tangentMode=RCTM_User, time=0.9978, value=1.46}
          ]
         }}}))
     2=<1 1>
     3=<1 1>
     4=<1 1>
     Index=&SpawnGroup()
     Num=::NewEnumerator3
    )
    Scale Sprite Size Mode=::NewEnumerator2
   )
   &ScaleMeshSize(Scale Factor=&MergeHelper_VectorSwitch(
     0=&VectorFromFloat(Value=&FloatFromCurve(FloatCurve=DataInterface() {propertyValues {Curve {
          defaultValue=0
          keys [
           {arriveTangent=2.613, interpMode=RCIM_Cubic, leaveTangent=2.613, tangentMode=RCTM_User, time=0, value=0}
           {arriveTangent=1.5358, interpMode=RCIM_Cubic, leaveTangent=1.5358, tangentMode=RCTM_User, time=0.1852, value=0.4616}
           {arriveTangent=0.0224, interpMode=RCIM_Cubic, leaveTangent=0.0224, tangentMode=RCTM_User, time=1, value=1}
          ]
         }}}))
     1=<0 0 0>
     2=<0 0 0>
     3=<0 0 0>
     4=<0 0 0>
     Index=&SpawnGroup()
     Num=::NewEnumerator3
    ))
   &ScaleColor(
    Scale Alpha=&MergeHelper_FloatSwitch(
     0=&FloatFromCurve(
      FloatCurve=DataInterface() {propertyValues {Curve {keys [
          <at -0.0046 1.0113 RCIM_Cubic>
          {arriveTangent=-1.125, interpMode=RCIM_Cubic, leaveTangent=-1.125, tangentMode=RCTM_User, time=0.3836, value=0.3261}
          <at 1.0011 0.002 RCIM_Cubic>
         ]}}}
      Scale Curve=&FN_FrontEndFloatModifier(Gameplay=0.75, Front End=0.4)
     )
     1=&FloatFromCurve(
      FloatCurve=DataInterface() {propertyValues {Curve {
         defaultValue=0
         keys [
          {arriveTangent=-3.0035, interpMode=RCIM_Cubic, leaveTangent=-3.0035, tangentMode=RCTM_User, time=0.0118, value=1.0048}
          {arriveTangent=-1.2935, interpMode=RCIM_Cubic, leaveTangent=-1.2935, tangentMode=RCTM_User, time=0.3271, value=0.3741}
          {arriveTangent=-0.0404, interpMode=RCIM_Cubic, leaveTangent=-0.0404, tangentMode=RCTM_User, time=1, value=0}
         ]
        }}}
      Scale Curve=&FN_FrontEndFloatModifier(
       Gameplay=&RandomRangeFloat(Minimum=2.5, Maximum=6)
       Front End=&RandomRangeFloat(Minimum=2.25, Maximum=4)
      )
     )
     2=1
     3=1
     4=1
     Index=&SpawnGroup()
     Num=::NewEnumerator3
    )
    Scale RGB=&MergeHelper_VectorSwitch(
     0=&MakeVectorFromLinearColorRGB(Color=&ColorFromCurve(ColorCurve=DataInterface() {propertyValues {
         AlphaCurve {keys []}
         BlueCurve {keys [<at 0.0953 1 RCIM_Cubic>, {arriveTangent=-1.6624, interpMode=RCIM_Cubic, time=0.6899, value=0.0449}, <at 0.9592 0.0149>]}
         GreenCurve {keys [<at 0.0953 1 RCIM_Cubic>, {arriveTangent=-1.5788, interpMode=RCIM_Cubic, time=0.6899, value=0.1939}, <at 0.9592 0.0644>]}
         RedCurve {keys [<at 0.0953 1 RCIM_Cubic>, {arriveTangent=-1.1268, interpMode=RCIM_Cubic, time=0.6899, value=1}, <at 0.9592 0.3323>]}
        }}))
     1=&MakeVectorFromLinearColorRGB(Color=&MakeCustomLinearColorFromBool(
       Bool=&Boolean_OrOperation(Bool1=$User.bIsFrontEnd, Bool2=$User.bIsFrontEndPreview)
       True Color=&ColorFromCurve(ColorCurve=DataInterface() {propertyValues {
          AlphaCurve {keys []}
          BlueCurve {keys [<at 0.1264 1.5>, {arriveTangent=-2.5049, interpMode=RCIM_Cubic, leaveTangent=-2.5049, time=0.2304, value=1}, <at 0.3261 1 RCIM_Cubic>]}
          GreenCurve {keys [<at 0.1264 1.5>, {arriveTangent=-2.5049, interpMode=RCIM_Cubic, leaveTangent=-2.5049, time=0.2304, value=1}, <at 0.3261 1 RCIM_Cubic>]}
          RedCurve {keys [<at 0.1264 1.5>, {arriveTangent=-2.5049, interpMode=RCIM_Cubic, leaveTangent=-2.5049, time=0.2304, value=1}, <at 0.3261 1 RCIM_Cubic>]}
         }})
       False Color=&ColorFromCurve(ColorCurve=DataInterface() {propertyValues {
          AlphaCurve {keys []}
          BlueCurve {keys [<at 0.1264 3>, {interpMode=RCIM_Cubic, leaveTangent=-2.2895, time=0.2304, value=1}, <at 0.3261 1 RCIM_Cubic>]}
          GreenCurve {keys [<at 0.1264 3>, {interpMode=RCIM_Cubic, leaveTangent=-2.2895, time=0.2304, value=1}, <at 0.3261 1 RCIM_Cubic>]}
          RedCurve {keys [<at 0.1264 3>, {interpMode=RCIM_Cubic, leaveTangent=-2.2895, time=0.2304, value=1}, <at 0.3261 1 RCIM_Cubic>]}
         }})
      ))
     2=<1 1 1>
     3=<1 1 1>
     4=<1 1 1>
     Index=&SpawnGroup()
     Num=::NewEnumerator3
    )
    ScaleA=true
    ScaleRGB=true
   )
   &Drag(
    Drag=&MergeHelper_FloatSwitch(0=0, 1=1, 2=0, 3=0, 4=0, Index=&SpawnGroup(), Num=::NewEnumerator3)
    Rotational Drag:NiagaraFloat=&MergeHelper_FloatSwitch(0=0, 1=1, 2=0, 3=0, 4=0, Index=&SpawnGroup(), Num=::NewEnumerator3)
    Use Linear Drag=true
    Use Rotational Drag=true
   )
   &SolveForcesAndVelocity(Clamp Velocity=false, Limit Acceleration=false)
  ]
  rendererClasses [&NiagaraMeshRendererProperties, &NiagaraSpriteRendererProperties]
  renderers [
   {rendererClass=&NiagaraMeshRendererProperties, rendererIndex=0}
   {rendererClass=&NiagaraSpriteRendererProperties, rendererIndex=1}
  ]
 }
 StarBits {
  emitterUpdateScript [
   &EmitterState()
   &SpawnBurst_Instantaneous(Spawn Count=11, Use Spawn Probability=false, UseLoopCountLimit=false)
   &SpawnBurst_Instantaneous(
    Spawn Count=&RandomRangeInt(Minimum=5, Maximum=7)
    Spawn Group=1
    Use Spawn Probability=false
    UseLoopCountLimit=false
   ) {module=SpawnBurst_Instantaneous001}
   &SpawnBurst_Instantaneous(Spawn Count=7, Spawn Group=2, Use Spawn Probability=false, UseLoopCountLimit=false) {module=SpawnBurst_Instantaneous002}
  ]
  particleSpawnScript [
   /SetVariables(Particles.VisibilityTag=&SpawnGroup()) {isSetParameters=true}
   &InitializeParticle(
    AdjustAlpha=false
    AdjustHue=true
    AdjustSaturation=false
    AdjustValue=false
    Color=&MergeHelper_ColorSwitch(
     0=<rgba 5 2.0805 0.2401 1>
     1=&RandomRangeLinearColor(Minimum=<rgba 10 7.2158 2.729 0.8>, Maximum=<rgba 15 9.8287 0.649 0.8>)
     2=<rgba 1 0.4454 0.0767 0.35>
     3=<rgba 0 0 0 0>
     4=<rgba 0 0 0 0>
     Index=&SpawnGroup()
     Num=::NewEnumerator3
    )
    Color Mode=::NewEnumerator3
    Hue Shift Range:Vector2f=&MergeHelper_Vector2Switch(0=<-0.05 0.05>, 1=<0 0>, 2=<-0.1 0.1>, 3=<0 0>, 4=<0 0>, Index=&SpawnGroup(), Num=::NewEnumerator3)
    Lifetime=&MergeHelper_FloatSwitch(
     0=&RandomRangeFloat(Minimum=0.3, Maximum=0.7)
     1=&RandomRangeFloat(Minimum=0.25, Maximum=0.5)
     2=&RandomRangeFloat(Minimum=0.2, Maximum=0.7)
     3=0
     4=0
     Index=0
     Num=::NewEnumerator3
    )
    Sprite Size Mode=::NewEnumerator3
    Uniform Sprite Size=&MergeHelper_FloatSwitch(
     0=&Multiply_Float(
      A=&RandomRangeFloat(Minimum=4, Maximum=11)
      B=&RandomRangeFloat(Minimum=0.9, Maximum=2)
     )
     1=&Multiply_Float(
      A=&RandomRangeFloat(Minimum=5, Maximum=15)
      B=&RandomRangeFloat(Minimum=1, Maximum=3)
     )
     2=&Multiply_Float(A=&RandomRangeFloat(Minimum=4, Maximum=16), B=$System.ScaleMultiplier)
     3=0
     4=0
     Index=&SpawnGroup()
     Num=::NewEnumerator3
    )
    UsePositionOffset=false
    Write Material Random=true
   )
   &CameraOffset(Camera Offset Amount=$System.OffsetCam)
   &ShapeLocation(
    MaskSpawnGroup=false
    Override Seed=false
    Sphere Radius=&MergeHelper_FloatSwitch(
     0=&Multiply_Float(A=15, B=&RandomRangeFloat(Minimum=1, Maximum=2))
     1=&Multiply_Float(A=22, B=&RandomRangeFloat(Minimum=1, Maximum=2))
     2=0
     3=0
     4=0
     Index=&SpawnGroup()
     Num=::NewEnumerator3
    )
    UseOwnerScale=true
   )
   &ShapeLocation(
    MaskSpawnGroup=false
    Sphere Distribution=::NewEnumerator3
    Sphere Radius=&MaskFloat(
     Mask Source=0
     Mask Parameter=?
     Mask=2
     Masked Float=&Multiply_Float(A=5, B=$System.ScaleMultiplier)
     Passthrough Float=0
    )
    UseOwnerScale=true
   ) {module=ShapeLocation001}
   &AddVelocity(
    Module.Use Falloff Distance=false
    Override Seed=false
    Velocity Mode=::NewEnumerator1
    Velocity Speed=&MergeHelper_FloatSwitch(
     0=&UniformAOrBFloat(
      A=&RandomRangeFloat(Minimum=125, Maximum=180)
      B=250
      Distribution Weight=0.3
      Evaluation Type=::NewEnumerator0
      Recalculate Random Each Loop=true
      Override Seed=false
     )
     1=&UniformAOrBFloat(
      A=&RandomRangeFloat(Minimum=35, Maximum=60)
      B=90
      Distribution Weight=0.35
      Evaluation Type=::NewEnumerator0
      Recalculate Random Each Loop=true
      Override Seed=false
     )
     2=&RandomRangeFloat(Minimum=20, Maximum=35)
     3=0
     4=0
     Index=&SpawnGroup()
     Num=::NewEnumerator3
    )
   )
   &ApplyInitialForces()
   /SetVariables(Particles.InitialCrossStarColor:LinearColor=&RandomRangeLinearColor(Minimum=<rgba 10 7.2158 2.729 0.8>, Maximum=<rgba 15 9.8287 0.649 0.8>)) {isSetParameters=true}
  ]
  particleUpdateScript [
   &ParticleState()
   &ScaleSpriteSize(
    Scale Sprite Size Mode=::NewEnumerator0
    Uniform Scale Factor=&MergeHelper_FloatSwitch(
     0=&Waveform(Global Phase=$Particles.NormalizedAge, "Waveform [1]"=::NewEnumerator2, "[1] Frequency"=1, "[1] Amplitude Min/Max"=<0.1 1>)
     1=1
     2=&FloatFromCurve(FloatCurve=DataInterface() {propertyValues {Curve {keys [{interpMode=RCIM_Cubic, leaveTangent=11.19, tangentMode=RCTM_Break, time=0, value=0}, {tangentMode=RCTM_User, time=0.2, value=1}]}}})
     3=1
     4=1
     Index=&SpawnGroup()
     Num=::NewEnumerator3
    )
   )
   &CurlNoiseForce(
    Bias Noise=false
    Noise Frequency=&MergeHelper_FloatSwitch(0=15, 1=0, 2=0, 3=0, 4=0, Index=&SpawnGroup(), Num=::NewEnumerator3)
    Noise Strength=&MergeHelper_FloatSwitch(
     0=&RandomRangeFloat(Minimum=60, Maximum=120)
     1=0
     2=0
     3=0
     4=0
     Index=&SpawnGroup()
     Num=::NewEnumerator3
    )
    Pan Noise=false
   )
   &GravityForce(Gravity=&MaskVector(Mask Source=0, Mask Parameter=?, Mask=1, Masked Vector=<0 0 -80>, Passthrough Vector=<0 0 0>))
   &Drag(
    Drag=&MergeHelper_FloatSwitch(
     0=&RandomRangeFloat(Minimum=6, Maximum=7)
     1=2.5
     2=2
     3=0
     4=0
     Index=&SpawnGroup()
     Num=::NewEnumerator3
    )
    Use Linear Drag=true
    Use Rotational Drag=true
   )
   &SolveForcesAndVelocity(Clamp Velocity=false, Limit Acceleration=false)
   &ScaleColor(
    Scale Alpha=&MergeHelper_FloatSwitch(
     0=1
     1=1
     2=&FloatFromCurve(FloatCurve=DataInterface() {propertyValues {Curve {
         defaultValue=1
         keys [
          {arriveTangent=9.0725, interpMode=RCIM_Cubic, leaveTangent=9.0725, tangentMode=RCTM_User, time=0, value=0}
          {arriveTangent=-2.5249, interpMode=RCIM_Cubic, leaveTangent=-2.5249, tangentMode=RCTM_User, time=0.1408, value=0.9847}
          {arriveTangent=-0.0405, interpMode=RCIM_Cubic, leaveTangent=-0.0405, tangentMode=RCTM_User, time=1, value=0}
         ]
        }}})
     3=0
     4=0
     Index=&SpawnGroup()
     Num=::NewEnumerator3
    )
    Scale RGB=&MergeHelper_VectorSwitch(
     0=<1 1 1>
     1=<1 1 1>
     2=&MakeVectorFromLinearColorRGB(Color=&ColorFromCurve(ColorCurve=DataInterface() {propertyValues {
         AlphaCurve {keys [<at 0 1 RCIM_Cubic>,<at 1 0 RCIM_Cubic>]}
         BlueCurve {keys [<at 0 1 RCIM_Cubic>,<at 1 1 RCIM_Cubic>]}
         GreenCurve {keys [<at 0 1 RCIM_Cubic>,<at 1 1 RCIM_Cubic>]}
         RedCurve {keys [<at 0 1 RCIM_Cubic>,<at 1 1 RCIM_Cubic>]}
        }}))
     3=<0 0 0>
     4=<0 0 0>
     Index=&SpawnGroup()
     Num=::NewEnumerator3
    )
    ScaleA=true
    ScaleRGB=true
   )
   /SetVariables(Particles.CrossStarsColor:LinearColor=&MultiplyLinearColor(
     Linear Color=$Particles.Initial.InitialCrossStarColor
     Scale Factor=&Vector4FromCurve(Vector4Curve=DataInterface() {propertyValues {
        AlphaCurve {keys [<at 0.0995 1>,<at 0.4733 0.4>,<at 1 0>]}
        BlueCurve {keys [<at 0 0>,<at 0.1557 1>,<at 0.6368 1>,<at 1 0>]}
        GreenCurve {keys [<at 0 5.2558>,<at 0.1557 1>,<at 0.6368 1>,<at 1 0.3697>]}
        RedCurve {keys [<at 0 10>,<at 0.1557 1>,<at 0.6368 1>,<at 1 1>]}
       }})
    )) {isSetParameters=true}
   /SetVariables(Particles.CrossStarsSize:Vector2f=&Multiply_Vector2D_ByFloat(
     Float=&Sine(
      Normalized Angle=$Particles.NormalizedAge
      Period=&RandomRangeFloat(Minimum=0.45, Maximum=0.8)
      Scale=1
      Bias=&RandomRangeFloat(Minimum=0, Maximum=1)
     )
     Vector2D=$Particles.Initial.SpriteSize
    )) {isSetParameters=true}
  ]
  rendererClasses [&NiagaraSpriteRendererProperties]
  renderers [
   {rendererClass=&NiagaraSpriteRendererProperties}
   {rendererClass=&NiagaraSpriteRendererProperties, rendererIndex=1}
   {rendererClass=&NiagaraSpriteRendererProperties, rendererIndex=2}
  ]
 }
}
summary {
 systemName=NS_Emote_BurgleBoss_PropSpawn
 emitters [
  {emitterName=Glow, rendererClasses [&NiagaraSpriteRendererProperties]}
  {emitterName=StarBits, rendererClasses [&NiagaraSpriteRendererProperties]}
  {emitterName=MainBurst, rendererClasses [&NiagaraMeshRendererProperties, &NiagaraSpriteRendererProperties]}
 ]
 userVariables [
  User.bIsFrontEnd {type=NiagaraBool}
  User.bIsFrontEndPreview {type=NiagaraBool}
 ]
}
